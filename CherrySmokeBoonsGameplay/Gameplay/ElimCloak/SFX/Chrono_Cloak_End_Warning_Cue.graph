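# Sound cue for the elimination cloak's expiry warning, routed to the gadgets sound class.
# An IsLocal branch sends the local player to a path with no attenuation asset assigned; others go to an IsEnemy branch.
# Enemy or unset plays at full volume and non-enemy much quieter, both using the cloak-activate attenuation.
# Each path randomly picks from the same four warning waves, with equal weights.
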SoundCue /CherrySmokeBoonsGameplay/Gameplay/ElimCloak/SFX/Chrono_Cloak_End_Warning_Cue
root=Branch_0
nodes [
 {id=Attenuation_0, children [Random_1]}
 {
  id=Attenuation_1
  children [Random_3]
  properties {attenuationSettings=&/Fort_Cloak_Activate_Att}
 }
 {
  id=Attenuation_2
  children [Random_4]
  properties {attenuationSettings=&Fort_Cloak_Activate_Att}
 }
 {id=Branch_0, children [Mixer_0,Branch_1,Branch_1], properties {boolParameterName=IsLocal}}
 {id=Branch_1, children [Mixer_1,Mixer_2,Mixer_1], properties {boolParameterName=IsEnemy}}
 {id=Mixer_0, children [Attenuation_0], properties {inputVolume [1]}}
 {id=Mixer_1, children [Attenuation_1], properties {inputVolume [1]}}
 {id=Mixer_2, children [Attenuation_2], properties {inputVolume [0.2]}}
 {
  id=Random_1
  children [WavePlayer_1,WavePlayer_5,WavePlayer_6,WavePlayer_7]
  properties {bShouldExcludeFromBranchCulling=true, weights [1,1,1,1]}
 }
 {
  id=Random_3
  children [WavePlayer_0, WavePlayer_13, WavePlayer_14, WavePlayer_15]
  properties {bShouldExcludeFromBranchCulling=true, weights [1,1,1,1]}
 }
 {
  id=Random_4
  children [WavePlayer_16, WavePlayer_17, WavePlayer_18, WavePlayer_19]
  properties {bShouldExcludeFromBranchCulling=true, weights [1,1,1,1]}
 }
 {id=WavePlayer_0, properties {soundWaveAssetPtr=&/Chrono_Cloak_Expire_Warning_01}}
 {id=WavePlayer_1, properties {soundWaveAssetPtr=&Chrono_Cloak_Expire_Warning_01}}
 {id=WavePlayer_13, properties {soundWaveAssetPtr=&/Chrono_Cloak_Expire_Warning_02}}
 {id=WavePlayer_14, properties {soundWaveAssetPtr=&/Chrono_Cloak_Expire_Warning_03}}
 {id=WavePlayer_15, properties {soundWaveAssetPtr=&/Chrono_Cloak_Expire_Warning_04}}
 {id=WavePlayer_16, properties {soundWaveAssetPtr=&Chrono_Cloak_Expire_Warning_01}}
 {id=WavePlayer_17, properties {soundWaveAssetPtr=&Chrono_Cloak_Expire_Warning_02}}
 {id=WavePlayer_18, properties {soundWaveAssetPtr=&Chrono_Cloak_Expire_Warning_03}}
 {id=WavePlayer_19, properties {soundWaveAssetPtr=&Chrono_Cloak_Expire_Warning_04}}
 {id=WavePlayer_5, properties {soundWaveAssetPtr=&Chrono_Cloak_Expire_Warning_02}}
 {id=WavePlayer_6, properties {soundWaveAssetPtr=&Chrono_Cloak_Expire_Warning_03}}
 {id=WavePlayer_7, properties {soundWaveAssetPtr=&Chrono_Cloak_Expire_Warning_04}}
]
properties {maxDistance=2097152, soundClassObject=&Fort_Gadgets, volumeMultiplier=1.2}
