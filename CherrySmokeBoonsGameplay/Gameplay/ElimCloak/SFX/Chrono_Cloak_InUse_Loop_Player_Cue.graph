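# Looping sound cue for an active Chrono cloak on a player, with separate mixes for the local player, enemies and non-enemies.
# IsLocal selects a first-person start sound layered with a delayed loop that fades in, then eases down.
# Otherwise IsEnemy picks the delayed third-person loop with Doppler, much quieter for non-enemies, defaulting to the enemy mix when unset.
# Velocity modulates loop pitch and volume, all branches are attenuated, and concurrency stops the quietest.
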SoundCue /CherrySmokeBoonsGameplay/Gameplay/ElimCloak/SFX/Chrono_Cloak_InUse_Loop_Player_Cue
root=Branch_0
nodes [
 {
  id=Attenuation_0
  children [Mixer_4]
  properties {attenuationSettings=&/Fort_Cloak_Ambient_Loop_Att}
 }
 {
  id=Attenuation_1
  children [Doppler_0]
  properties {attenuationSettings=&Fort_Cloak_Ambient_Loop_Att}
 }
 {
  id=Attenuation_2
  children [Doppler_1]
  properties {attenuationSettings=&Fort_Cloak_Ambient_Loop_Att}
 }
 {id=Branch_0, children [Mixer_1,Branch_1,Branch_1], properties {boolParameterName=IsLocal}}
 {id=Branch_1, children [Mixer_2,Mixer_3,Mixer_2], properties {boolParameterName=IsEnemy}}
 {id=Delay_2, children [WavePlayer_2], properties {delayMax=1, delayMin=1}}
 {id=Delay_3, children [WavePlayer_3], properties {delayMax=1, delayMin=1}}
 {id=Delay_4, children [WavePlayer_4], properties {delayMax=1, delayMin=1}}
 {id=Doppler_0, children [ModulatorContinuous_2], properties {bUseSmoothing=true, dopplerIntensity=10, smoothingInterpSpeed=2}}
 {id=Doppler_1, children [ModulatorContinuous_3], properties {bUseSmoothing=true, dopplerIntensity=10, smoothingInterpSpeed=2}}
 {
  id=Enveloper_1
  children [Delay_4]
  properties {pitchCurve {editorCurveData {keys [{time=0}]}}, volumeCurve {editorCurveData {keys [<at 0 0>,<at 1 1>,<at 1.2691 1>,<at 15 0.35>]}}}
 }
 {id=Mixer_1, children [Attenuation_0], properties {inputVolume [1.25]}}
 {id=Mixer_2, children [Attenuation_1], properties {inputVolume [1.3]}}
 {id=Mixer_3, children [Attenuation_2], properties {inputVolume [0.2]}}
 {id=Mixer_4, children [ModulatorContinuous_4, Modulator_1], properties {inputVolume [1,1.3]}}
 {
  id=ModulatorContinuous_2
  children [Delay_2]
  properties {
   pitchModulationParams {default=0, maxOutput=1.25, minOutput=0.8, parameterName=Velocity}
   volumeModulationParams {default=0, minOutput=0.5, parameterName=Velocity}
  }
 }
 {
  id=ModulatorContinuous_3
  children [Delay_3]
  properties {
   pitchModulationParams {default=0, maxOutput=1.25, minOutput=0.8, parameterName=Velocity}
   volumeModulationParams {default=0, minOutput=0.5, parameterName=Velocity}
  }
 }
 {
  id=ModulatorContinuous_4
  children [Enveloper_1]
  properties {
   pitchModulationParams {default=0, maxOutput=1.25, minOutput=0.8, parameterName=Velocity}
   volumeModulationParams {default=0, minOutput=0.5, parameterName=Velocity}
  }
 }
 {id=Modulator_1, children [WavePlayer_5]}
 {id=WavePlayer_2, properties {bLooping=true, soundWaveAssetPtr=&/Chrono_Cloak_Active_Loop}}
 {id=WavePlayer_3, properties {bLooping=true, soundWaveAssetPtr=&Chrono_Cloak_Active_Loop}}
 {id=WavePlayer_4, properties {bLooping=true, soundWaveAssetPtr=&/Chrono_Cloak_Active_Loop_1P}}
 {id=WavePlayer_5, properties {soundWaveAssetPtr=&/Chrono_Cloak_Active_Loop_Start_1P}}
]
properties {
 bOverrideConcurrency=true
 maxDistance=5300
 soundClassObject=&Fort_Gadgets
 volumeMultiplier=1
 concurrencyOverrides {maxCount=8, resolutionRule=StopQuietest}
}
