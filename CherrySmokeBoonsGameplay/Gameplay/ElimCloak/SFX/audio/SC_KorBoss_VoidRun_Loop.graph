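# Sound cue for the Void Run loop audio (KorBoss naming) in the ElimCloak boon's SFX.
# IsLocal and IsEnemy branches set mix level: local loudest with its own attenuation node, then enemy or unset, then friendly.
# All paths share one Enveloper that fades volume in, randomizes volume and pitch, and layers two weighted loop waves.
# The cue overrides concurrency to cap instances, stopping the quietest.
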SoundCue /CherrySmokeBoonsGameplay/Gameplay/ElimCloak/SFX/audio/SC_KorBoss_VoidRun_Loop
root=Branch_0
nodes [
 {
  id=Attenuation_0
  children [Mixer_0]
  properties {attenuationSettings=&/Att_KorBoss_VoidRun_Loop}
 }
 {id=Branch_0, children [Attenuation_0, Branch_1, Branch_1], properties {boolParameterName=IsLocal}}
 {id=Branch_1, children [Mixer_1,Mixer_2,Mixer_1], properties {boolParameterName=IsEnemy}}
 {
  id=Enveloper_1
  children [Mixer_3]
  properties {
   pitchMax=1.2
   pitchMin=1.1
   volumeMax=0.55
   volumeMin=0.45
   pitchCurve {editorCurveData {keys [{time=0}]}}
   volumeCurve {editorCurveData {keys [<at 0.1142 0.0291>,<at 1.9443 0.956>,<at 4.9191 0.8076>]}}
  }
 }
 {id=Mixer_0, children [Enveloper_1], properties {inputVolume [1]}}
 {id=Mixer_1, children [Enveloper_1], properties {inputVolume [0.9]}}
 {id=Mixer_2, children [Enveloper_1], properties {inputVolume [0.8]}}
 {id=Mixer_3, children [WavePlayer_0,WavePlayer_1], properties {inputVolume [0.6,0.4]}}
 {id=WavePlayer_0, properties {soundWaveAssetPtr=&/KorBoss_VoidRun_Loop_A}}
 {id=WavePlayer_1, properties {soundWaveAssetPtr=&/KorBoss_VoidRun_Loop_B}}
]
properties {
 attenuationSettings=&Att_KorBoss_VoidRun_Loop
 bOverrideConcurrency=true
 maxDistance=5300
 soundClassObject=&Fort_Gadgets_Items
 volumeMultiplier=0.5
 concurrencyOverrides {maxCount=8, resolutionRule=StopQuietest}
}
